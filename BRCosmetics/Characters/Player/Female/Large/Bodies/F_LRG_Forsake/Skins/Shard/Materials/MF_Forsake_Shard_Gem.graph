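# Material function for the Forsake Shard skin that blends a faceted gem look into incoming material attributes by a SkinFX mask channel, behind a static switch.
# Two tiled crystal normal layers form the facets, the inner one parallaxed with BumpOffset using noise height weighted toward grazing angles.
# Sun-direction, camera-direction and Fresnel terms drive a color-atlas curve and emissive glints, while the gem also overrides base color, metallic, specular, roughness and normal.
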MaterialFunction /BRCosmetics/Characters/Player/Female/Large/Bodies/F_LRG_Forsake/Skins/Shard/Materials/MF_Forsake_Shard_Gem
graph {nodes [
  n1 = Multiply(A=TextureCoordinate(), B=ScalarParameter() {defaultValue=4.75, group=Forsake Shard, parameterName=Normal_A_Tile})
  n2 = TextureSample(UVs=@n1) {samplerType=SAMPLERTYPE_LinearColor, texture=/Game/Characters/Player/Female/Medium/Bodies/F_MED_Shapeshifter/Textures/T_Crystal_Normal}
  n3 = ConstantBiasScale(None=ComponentMask(None=@n2.RGB) {g=true, r=true}) {bias=-0.5, scale=2}
  n4 = TextureSample(UVs=@n1) {samplerSource=SSM_Wrap_WorldGroupSettings, texture=/BRCosmetics/Creative/Devices/Generic/Materials/Textures/T_CP_Device_FortNoise_02_D}
  n5 = VertexNormalWS()
  n6 = CameraVectorWS()
  n7 = OneMinus(None=Saturate(None=DotProduct(A=@n5, B=@n6)))
  n8 = TextureCoordinate()
  n9 = ScalarParameter() {defaultValue=3.5, group=Forsake Shard, parameterName=Normal_B_Tile}
  n10 = TextureSample(UVs=MaterialFunctionCall!(
    HeightTexture=Multiply(A=@n4.G, B=@n7)
    TextureCoordinates=Multiply(A=@n8, B=@n9)
    ReferencePlane
    HeightRatioInput=ScalarParameter() {defaultValue=0.1, group=Forsake Shard, parameterName=CrystalDepth}
    CameraVector
   ) {materialFunction=/Engine/Functions/Engine_MaterialFunctions02/Utility/BumpOffset_Advanced.BumpOffset_advanced}.UVs) {samplerType=SAMPLERTYPE_LinearColor, texture=&T_Crystal_Normal}
  n11 = ConstantBiasScale(None=ComponentMask(None=@n10.RGB) {g=true, r=true}) {bias=-0.5, scale=2}
  n12 = Add(
   A=Multiply(
    A=ComponentMask(None=@n3) {g=true, r=true}
    B=ScalarParameter() {defaultValue=0.25, group=Forsake Shard, parameterName=Facet_A_Normal_Strength}
   )
   B=Multiply(
    A=ComponentMask(None=@n11) {g=true, r=true}
    B=ScalarParameter() {defaultValue=0.075, group=Forsake Shard, parameterName=Facet_B_Normal_Strength}
   )
  )
  AppendVector(A=@n12, B=Constant() {r=1})
  AtmosphericLightVector()
  Custom() {code="return View.DirectionalLightDirection;", description=DirectionalLightVector, inputs []}
  n13 = GetMaterialAttributes(None=FunctionInput() {inputName=MA In, inputType=FunctionInput_MaterialAttributes}) {attributeGetTypes [
    #1
    #2
    #3
    #4
    #5
    #6
   ]}
  n14 = NamedRerouteDeclaration(None=@n13.MaterialAttributes) {name=MA_In, nodeColor=<rgba 1 0.1641 0 1>}
  n15 = NamedRerouteUsage()
  n16 = TextureSampleParameter2D() {
   parameterName=SkinFX_Mask
   samplerSource=SSM_Wrap_WorldGroupSettings
   samplerType=SAMPLERTYPE_Masks
   texture=/BRCosmetics/Characters/Player/Female/Large/Bodies/F_LRG_Forsake/Skins/Shard/Textures/T_F_LRG_Forsake_Shard_Body_FX
  }
  n17 = StaticComponentMaskParameter(None=@n16.RGB) {defaultB=true, group=Forsake Shard, parameterName=Gem_SkinFX_Mask_Channel}
  FunctionOutput(None=StaticSwitchParameter(True=BlendMaterialAttributes(A=@n14, B=@n15, Alpha=@n17), False=@n14) {group=Forsake Shard, parameterName=Use_Forsake_Gem})
  n18 = LinearInterpolate(
   A=VectorParameter() {group=Forsake Shard, parameterName=Gem Base Color A, defaultValue=<rgba 0.0501 0.1247 0.13 1>}.RGB
   B=VectorParameter() {group=Forsake Shard, parameterName=Gem Base Color B, defaultValue=<rgba 0.1226 0.3073 0.2888 1>}.RGB
   Alpha=NamedRerouteUsage()
  )
  n19 = Constant()
  n20 = CollectionParameter() {
   collection=&FortniteMaterialParameters
   parameterName=SunlightDirectionalVector
  }
  n21 = DotProduct(A=ComponentMask(None=@n20) {b=true, g=true, r=true}, B=NamedRerouteUsage())
  n22 = Abs(None=@n21)
  n23 = Power(Base=@n22)
  n24 = NamedRerouteUsage()
  n25 = Add(A=NamedRerouteUsage(), B=AppendVector(A=ComponentMask(None=@n24) {g=true, r=true}, B=Constant()))
  n26 = DotProduct(
   A=MaterialFunctionCall() {materialFunction=&CameraDirectionVector}.Result
   B=Transform(None=Normalize(VectorInput=@n25))
  )
  n27 = Saturate(None=Power(Base=Cosine(None=@n26)) {constExponent=8})
  n28 = Multiply(A=@n27) {constB=0.25}
  n29 = CurveAtlasRowParameter(CurveTime=Saturate(None=Add(A=@n23, B=@n28))) {
   atlas=/BRCosmetics/Characters/Player/Female/Large/Bodies/F_LRG_Forsake/Skins/Shard/Textures/Forsake_Shard_ColorAtlas
   curve=/BRCosmetics/Characters/Player/Female/Large/Bodies/F_LRG_Forsake/Skins/Shard/Textures/Forsake_Shard_Gem_Colors
   group=Forsake Shard
   parameterName=Gem_Color
  }
  n30 = Sign(None=@n21)
  n31 = Multiply(A=Power(Base=@n22) {constExponent=6}, B=LinearInterpolate(Alpha=Saturate(None=@n30)) {constA=0.25})
  n32 = Fresnel(Normal=NamedRerouteUsage()) {baseReflectFraction=0, exponent=3}
  n33 = Max(A=Add(A=Add(A=@n31, B=@n27), B=@n32)) {constB=0.1}
  n34 = Subtract(A=@n32, B=NamedRerouteUsage())
  n35 = Max(A=Multiply(A=Saturate(None=@n34)) {constB=2}) {constB=0.1}
  n36 = NamedRerouteUsage()
  n37 = Multiply(A=@n2.R) {constB=0.25}
  n38 = SetMaterialAttributes!(
   MaterialAttributes=NamedRerouteUsage()
   BaseColor=@n18
   Metallic=ScalarParameter() {defaultValue=0.8, parameterName=Fracture_Metallic}
   Specular=ScalarParameter() {defaultValue=0.75, group=Forsake Shard, parameterName=Fracture_Specular}
   Normal=Add(A=AppendVector(A=@n12, B=@n19), B=NamedRerouteUsage())
   EmissiveColor=Add(A=Multiply(A=@n29, B=@n33), B=Multiply(A=@n18, B=@n35))
   Roughness=Saturate(None=Subtract(A=@n36, B=@n37))
  ) {
   desc=Gem Shader
   attributeSetTypes [
    #1
    #2
    #3
    #4
    #5
    #6
   ]
  }
  NamedRerouteDeclaration(None=@n38) {name=Gem_MA_Out, nodeColor=<rgba 0 0.8125 1 1>}
  n39 = PixelNormalWS()
  n40 = NamedRerouteDeclaration(None=@n11) {name=Inner_Fracture_Normals, nodeColor=<rgba 0 1 0.2031 1>}
  n41 = Add(
   A=Multiply(A=@n3, B=ScalarParameter() {defaultValue=0.2, group=Forsake Shard, parameterName=Facet_A_Strength})
   B=Multiply(A=@n40, B=ScalarParameter() {defaultValue=0.075, group=Forsake Shard, parameterName=Facet_B_Strength})
  )
  n42 = Transform(None=AppendVector(A=ComponentMask(None=@n41) {g=true, r=true}, B=Constant() {r=1}))
  NamedRerouteDeclaration(None=Normalize(VectorInput=Add(A=@n39, B=@n42))) {name=Combined_Fracture_Normals, nodeColor=<rgba 0 0.7422 1 1>}
  NamedRerouteDeclaration(None=@n13.Roughness) {name=Roughness_In, nodeColor=<rgba 1 0.7734 0 1>}
  NamedRerouteDeclaration(None=@n13.BaseColor) {name=BaseColor_In, nodeColor=<rgba 1 0.9375 0 1>}
  NamedRerouteDeclaration(None=@n13.Metallic) {name=Metallic_In, nodeColor=<rgba 0.9922 0 1 1>}
  NamedRerouteDeclaration(None=@n13.Specular) {name=Specular_In, nodeColor=<rgba 0.2422 0 1 1>}
  NamedRerouteDeclaration(None=@n13.Normal) {name=Normal_In, nodeColor=<rgba 1 0.3984 0 1>}
  NamedRerouteDeclaration(None=@n13.EmissiveColor) {name=Emissive_In, nodeColor=<rgba 1 0 0.6094 1>}
  n43 = MaterialFunctionCall(Float2=@n40) {materialFunction=&BreakOutFloat2Components}
  n44 = Add(A=@n43.R, B=@n43.R)
  NamedRerouteDeclaration(None=Saturate(None=Multiply(A=@n44) {constB=0.5})) {name=GemFracture_Mask, nodeColor=<rgba 0.4531 0 1 1>}
 ]}
